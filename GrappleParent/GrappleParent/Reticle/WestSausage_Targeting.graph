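# HUD element showing the targeting reticle for dual-swing grapple weapons whose reticle tag matches its own.
# It follows pawn and weapon changes, binds to the weapon's reticle updates and tracks in-vehicle state through async gameplay tag waits.
# Each update shows it only when locked on, usable, unattached, unblocked, on foot and not DBNO.
# TargetWidget sits at screen center or the projected target's center-relative position, and displays charge progress when requested.
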WidgetBlueprint /GrappleParent/GrappleParent/Reticle/WestSausage_Targeting
graphs [
 EventGraph {nodes [
   n1 = BP_ReticleUpdated!() -> (
    OutputDelegate
    then
    IsLockedOn:bool=~
    Position:Vector=~
    AbleToUse:bool=~
    ShowChargeUpCircle:bool=~
    ChargeTime:double=~
    IsTargetAtAimLocation:bool=~
    IsSwingAttached:bool=~
    IsFiringBlocked:bool=~
   ) {is=title K2Node_CustomEvent}
   > Set LockedOn(LockedOn:bool=@n1.IsLockedOn) -> (Output_Get:bool)
   > Set Attached(Attached:bool=@n1.IsSwingAttached) -> (Output_Get:bool)
   > Set AbleToUse(AbleToUse:bool=@n1.AbleToUse) -> (Output_Get:bool)
   > Set IsFiringBlocked(IsFiringBlocked:bool=@n1.IsFiringBlocked) -> (Output_Get:bool)
   > n2 = Branch(Condition=ShouldBeVisible!() -> (bShouldBeVisible:bool))
   > SetVisibility(InVisibility=HitTestInvisible)
   n3 = Get TargetWidget() -> (TargetWidget:RE West Sausage Targeting Object Reference)
   > SetPosition!(
    execute
    self:Canvas Panel Slot Object Reference=SlotAsCanvasSlot!(Widget:Widget Object Reference=@n3) -> (ReturnValue:Canvas Panel Slot Object Reference)
    InPosition:Vector 2D Structure=WorldToScreen!(WorldLocation:Vector=@n1.Position, bIsTargetAtWeaponAimLoc:bool=@n1.IsTargetAtAimLocation) -> (ScreenLocation:Vector 2D Structure)
   ) -> (then)
   > Branch(Condition=@n1.ShowChargeUpCircle)
   > UpdateChargeCircle!(execute, self:RE West Sausage Targeting Object Reference=@n3, Scale X:double=@n1.ChargeTime) -> (then)
   SetVisibility(execute=@n2.else, InVisibility=Collapsed)
   n4 = OnWeaponEquipped!() -> (OutputDelegate, then, NewWeapon:Fort Weapon Object Reference, PrevWeapon:Fort Weapon Object Reference) {is=title K2Node_CustomEvent}
   > n5 = Cast To FortWeaponRangedDualSwing!(execute, Object=@n4.NewWeapon) -> (then, CastFailed, AsFort Weapon Ranged Dual Swing:Fort Weapon Ranged Dual Swing Object Reference) {is=title K2Node_DynamicCast}
   > n6 = Branch(Condition="Equal (GameplayTag)"(
     A=Get ReticleTypeTag!(self:Fort Weapon Ranged Dual Swing Object Reference=@n5) -> (ReticleTypeTag:Gameplay Tag Structure)
     B=Get MyReticleTag() -> (MyReticleTag:Gameplay Tag Structure)
    ))
   > n7 = Set Weapon(Weapon:Fort Weapon Ranged Dual Swing Object Reference=@n5) -> (Output_Get:Fort Weapon Ranged Dual Swing Object Reference)
   > SetVisibility(InVisibility=HitTestInvisible)
   > Bind Event to BP Reticle Update(self:Fort Weapon Ranged Dual Swing Object Reference=@n7, Delegate=Create Event()) {is=title K2Node_AddDelegate}
   SetVisibility(execute=[@n5.CastFailed, @n6.else], InVisibility=Collapsed)
   OnPawnChanged() {is=title K2Node_CustomEvent}
   > n8 = Sequence()
   > n9 = Get!(execute:Exec) -> (then:Exec, else:Exec, AddInVehicleAsyncAction:Cancellable Async Action Object Reference)
   > n10 = Cancel(self:Cancellable Async Action Object Reference=@n9)
   n11 = Get!(execute=[@n10.then, @n9.else]) -> (then, else, RemoveInVehicleAsyncAction:Cancellable Async Action Object Reference)
   > n12 = Cancel(self=@n11)
   n13 = Get!(execute=[@n12.then, @n11.else]) -> (then, else, PlayerPawn:Fort Player Pawn Object Reference)
   n14 = Create Event()
   > Unbind Event from On Weapon Equipped Delegate(self:Fort Pawn Object Reference=@n13, Delegate=@n14) {is=title K2Node_RemoveDelegate}
   n15 = Cast To FortPlayerPawn(execute=@n8.then_1, Object=GetOwningPlayerPawn!(self:User Widget Object Reference) -> (ReturnValue:Pawn Object Reference))
   > n16 = Set PlayerPawn(PlayerPawn:Fort Player Pawn Object Reference=@n15) -> (Output_Get:Fort Player Pawn Object Reference)
   > OnWeaponEquipped!(
    execute
    self
    NewWeapon:Fort Weapon Object Reference=GetCurrentWeapon!(self:Fort Pawn Object Reference=@n16) -> (ReturnValue:Fort Weapon Object Reference)
    PrevWeapon:Fort Weapon Object Reference
   ) -> (then) {is=title K2Node_CallFunction}
   > Bind Event to On Weapon Equipped Delegate(self:Fort Pawn Object Reference=@n15, Delegate=@n14) {is=title K2Node_AddDelegate}
   > n17 = WaitGameplayTagAddToActor!(
    execute:Exec
    TargetActor:Actor Object Reference=@n15
    Tag:Gameplay Tag Structure=(TagName="Gameplay.Action.Player.AthenaInVehicle")
    OnlyTriggerOnce:bool=false
   ) -> (then:Exec, AsyncTaskProxy:Ability Async Wait Gameplay Tag Added Object Reference, Added:Exec) {is=title K2Node_AsyncAction}
   > Set AddInVehicleAsyncAction(AddInVehicleAsyncAction:Cancellable Async Action Object Reference=@n17) -> (Output_Get:Cancellable Async Action Object Reference)
   > n18 = WaitGameplayTagRemoveFromActor!(
    execute:Exec
    TargetActor:Actor Object Reference=@n15
    Tag:Gameplay Tag Structure=(TagName="Gameplay.Action.Player.AthenaInVehicle")
    OnlyTriggerOnce:bool=false
   ) -> (then:Exec, AsyncTaskProxy:Ability Async Wait Gameplay Tag Removed Object Reference, Removed:Exec) {is=title K2Node_AsyncAction}
   > Set RemoveInVehicleAsyncAction(RemoveInVehicleAsyncAction:Cancellable Async Action Object Reference=@n18) -> (Output_Get:Cancellable Async Action Object Reference)
   Set bIsPawnInVehicle(execute=@n18.Removed, bIsPawnInVehicle:bool=false) -> (Output_Get:bool)
   Set bIsPawnInVehicle(execute=@n17.Added, bIsPawnInVehicle=true)
   Event Construct()
   > SetVisibility(InVisibility=Collapsed)
   > Bind Event to On Pawn Set(
    self:Fort HUDContext Object Reference=Get FortHUDContext!() -> (ReturnValue:Fort HUDContext Object Reference) {is=title K2Node_GetBlueprintContextEx}
    Delegate=Create Event()
   ) {is=title K2Node_AddDelegate}
   > OnPawnChanged() {is=title K2Node_CallFunction}
  ]}
 ShouldBeVisible {nodes [
   ShouldBeVisible() {is=title K2Node_FunctionEntry}
   > n1 = Get()
   n2 = Get IsFiringBlocked() -> (IsFiringBlocked:bool)
   n3 = IsDBNO!(self:Fort Pawn Object Reference=@n1) -> (ReturnValue:bool)
   n4 = Get Attached() -> (Attached:bool)
   n5 = Get bIsPawnInVehicle() -> (bIsPawnInVehicle:bool)
   > Return Node!(
    execute
    bShouldBeVisible:bool=AND Boolean!(
     A=NOT Boolean(A=@n5)
     B=Get AbleToUse() -> (AbleToUse:bool)
     C:bool=Get LockedOn() -> (LockedOn:bool)
     D:bool=NOT Boolean(A=@n3)
     E:bool=NOT Boolean(A=@n4)
     F:bool=NOT Boolean(A=@n2)
    ) -> (ReturnValue)
   )
   Return Node(execute=@n1.else, bShouldBeVisible=false)
  ]}
 WorldToScreen {nodes [
   n1 = WorldToScreen!() -> (then, WorldLocation:Vector=~, bIsTargetAtWeaponAimLoc:bool=~) {is=title K2Node_FunctionEntry}
   > n2 = Branch(Condition=@n1.bIsTargetAtWeaponAimLoc)
   > Return Node!(execute, ScreenLocation:Vector 2D Structure)
   SetWidgetCZonesAutomation!(execute=@n2.else, bEnabled:bool=true) -> (then)
   > Set TmpPos(TmpPos:Vector 2D Structure=ProjectToWidgetPosition!(PlayerController:Player Controller Object Reference=GetOwningPlayer(), WorldLocation:Vector=@n1, bPlayerViewportRelative:bool=false) -> (ScreenPosition:Vector 2D Structure, ReturnValue:bool)) -> (Output_Get:Vector 2D Structure)
   > SetWidgetCZonesAutomation(bEnabled=false)
   n3 = BreakVector2D(InVec=GetViewportSize!() -> (ReturnValue:Vector 2D Structure))
   n4 = float / float(A=@n3, B=2)
   n5 = float / float(A=@n3.Y, B=2)
   > Return Node(ScreenLocation=vector2d - vector2d(A:Vector 2D Structure=Get TmpPos() -> (TmpPos:Vector 2D Structure), B:Vector 2D Structure=MakeVector2D(X=@n4, Y=@n5)) -> (ReturnValue:Vector 2D Structure) {is=title K2Node_PromotableOperator})
  ]}
]
variables [
 AbleToUse {type=boolean}
 AddInVehicleAsyncAction {type=CancellableAsyncAction, default=null}
 Attached {type=boolean}
 bIsPawnInVehicle {type=boolean}
 IsFiringBlocked {type=boolean}
 LockedOn {type=boolean}
 MyReticleTag {type=GameplayTag, default {tagName=HUD.Athena.GrappleParent.DefaultReticle}}
 PlayerPawn {type=FortPlayerPawn}
 RemoveInVehicleAsyncAction {type=CancellableAsyncAction, default=null}
 Weapon {type=FortWeaponRangedDualSwing, default=null}
]
widgets {
 info {inheritedWidgetCount=1, parentClass=&FortHUDElementWidget, rootWidgetClass=&CanvasPanel, widgetCount=3}
 widgets [
  CanvasPanel_275 {widgetClassPath=&CanvasPanel, widget=CanvasPanel_275}
  TargetWidget {
   widgetClassPath=&/RE_WestSausage_Targeting
   bIsVariable=true
   parent=CanvasPanel_275
   widget=TargetWidget
   slot {
    class=CanvasPanelSlot
    properties {layoutData {alignment=<0.5 0.5>, anchors {maximum=<0.5 0.5>, minimum=<0.5 0.5>}, offsets=<ltrb 0 0 30 30>}}
   }
  }
  TutorialHighlight {widgetClassPath=&TutorialHighlight}
 ]
}
