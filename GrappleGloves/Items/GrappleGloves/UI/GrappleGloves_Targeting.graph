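# Targeting reticle widget for the Grapple Gloves, derived from the WestSausage_Targeting widget in GrappleParent.
# It adds no widgets or variables of its own, and its Construct event only calls the parent implementation.
# A leftover debug Print String node sits unconnected in the event graph.
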WidgetBlueprint /GrappleGloves/Items/GrappleGloves/UI/GrappleGloves_Targeting
graphs [EventGraph {nodes [
   Event Construct()
   > Parent: Construct()
   PrintString!(execute, InString="DO IT!!!!!", bPrintToScreen, bPrintToLog, TextColor=(R=0.0941,G=1,B=0,A=1), Duration=50, Key) -> (then)
  ]}]
variables []
widgets {
 info {
  inheritedWidgetCount=0
  parentClass=/GrappleParent/GrappleParent/Reticle/WestSausage_Targeting.WestSausage_Targeting_C
  rootWidgetClass=null
  widgetCount=0
 }
 widgets []
}
